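# Animation Blueprint for the crow prey animal, driving ground idle and flight locomotion through one state machine with inertialized output.
# A conduit routes to idle or to takeoff when bIsFlying is set; takeoff plays into flying, and bIsLanding triggers landing back to idle.
# Flying switches between glide and rate-scaled flap loops, layering additive pitch poses, with flap pose chosen by bIsFlyingSlow.
# Driving variables are not declared here, and the event graph holds no logic.
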AnimBlueprint /Irwin/AI/Prey/Crow/Animations/AnimBP_Fauna_Crow
target_skeleton=/Irwin/AI/Prey/Crow/Meshes/Crow_Bird_Skeleton
graphs [
 AnimGraph {nodes [
   AnimGraphNode_StateMachine_1 = CrowLoco() {is=title AnimGraphNode_StateMachine}
   Output Pose(Result=Inertialization!(Source:Pose Link Structure=@AnimGraphNode_StateMachine_1) -> (Pose:Pose Link Structure) {is=title AnimGraphNode_Inertialization}) {layerGroup=None}
  ]}
 CrowLoco {
  entry=PassThrough
  path=AnimGraph.AnimGraphNode_StateMachine_1.CrowLoco
  ids=id
  nodes [
   AnimStateConduitNode_0 = Conduit()
   AnimStateNode_0 = PassThrough() {is=title AnimStateNode}
   AnimStateNode_1 = Idle()
   AnimStateNode_2 = Land() {is=title AnimStateNode}
   AnimStateNode_3 = Flying() {is=title AnimStateNode}
   AnimStateNode_4 = TakeOff() {is=title AnimStateNode}
   AnimStateTransitionNode_0 = AnimStateTransitionNode()
   AnimStateTransitionNode_1 = AnimStateTransitionNode() {logicType=TLT_Inertialization}
   AnimStateTransitionNode_3 = AnimStateTransitionNode() {crossfadeDuration=0.1, logicType=TLT_Inertialization}
   AnimStateTransitionNode_4 = AnimStateTransitionNode() {bAutomaticRuleBasedOnSequencePlayerInState=true, crossfadeDuration=0.3, logicType=TLT_Inertialization}
   AnimStateTransitionNode_5 = AnimStateTransitionNode()
   AnimStateTransitionNode_6 = AnimStateTransitionNode() {crossfadeDuration=0.1, logicType=TLT_Inertialization}
   AnimStateTransitionNode_7 = AnimStateTransitionNode() {bAutomaticRuleBasedOnSequencePlayerInState=true, crossfadeDuration=1.5}
  ]
  links [
   {
    dst=Idle
    rule=AnimGraph.AnimGraphNode_StateMachine_1.CrowLoco.AnimStateTransitionNode_1.Transition
    src=Conduit
    via=AnimStateTransitionNode_1
   }
   {
    dst=TakeOff
    rule=AnimGraph.AnimGraphNode_StateMachine_1.CrowLoco.AnimStateTransitionNode_6.Transition
    src=Conduit
    via=AnimStateTransitionNode_6
   }
   {
    dst=Land
    rule=AnimGraph.AnimGraphNode_StateMachine_1.CrowLoco.AnimStateTransitionNode_3.Transition
    src=Flying
    via=AnimStateTransitionNode_3
   }
   {
    dst=Conduit
    rule=AnimGraph.AnimGraphNode_StateMachine_1.CrowLoco.AnimStateTransitionNode_5.Transition
    src=Idle
    via=AnimStateTransitionNode_5
   }
   {dst=Idle, rule=AnimGraph.AnimGraphNode_StateMachine_1.CrowLoco.AnimStateTransitionNode_4.Transition, src=Land, via=AnimStateTransitionNode_4}
   {
    dst=Conduit
    rule=AnimGraph.AnimGraphNode_StateMachine_1.CrowLoco.AnimStateTransitionNode_0.Transition
    src=PassThrough
    via=AnimStateTransitionNode_0
   }
   {
    dst=Flying
    rule=AnimGraph.AnimGraphNode_StateMachine_1.CrowLoco.AnimStateTransitionNode_7.Transition
    src=TakeOff
    via=AnimStateTransitionNode_7
   }
  ]
 }
 Conduit {path=AnimGraph.AnimGraphNode_StateMachine_1.CrowLoco.AnimStateConduitNode_0.Conduit, nodes [Result(bCanEnterTransition=true)]}
 PassThrough {
  path=AnimGraph.AnimGraphNode_StateMachine_1.CrowLoco.AnimStateNode_0.PassThrough
  nodes [Output Animation Pose(Result=Local Space Ref Pose())]
 }
 Idle {
  path=AnimGraph.AnimGraphNode_StateMachine_1.CrowLoco.AnimStateNode_1.Idle
  nodes [Output Animation Pose(Result=Sequence Player 'Crow_Idle'() {sequence=&/Crow_Idle})]
 }
 Land {
  path=AnimGraph.AnimGraphNode_StateMachine_1.CrowLoco.AnimStateNode_2.Land
  nodes [Output Animation Pose(Result=Sequence Player 'Crow_Fly_Land'() {sequence=&/Crow_Fly_Land})]
 }
 Flying {
  path=AnimGraph.AnimGraphNode_StateMachine_1.CrowLoco.AnimStateNode_3.Flying
  nodes [
   n1 = Blend Poses by bool(
    BlendPose_0="Sequence Player (additive) 'Crow_Fly_Pose_Center'"() {sequence=&/Crow_Fly_Pose_Center}
    BlendPose_1="Sequence Player (additive) 'Crow_Fly_Pose_Up'"() {sequence=&/Crow_Fly_Pose_Up}
    BlendTime_0=0.3
    BlendTime_1=0.3
    bActiveValue=Get bIsFlyingSlow() -> (bIsFlyingSlow:bool)
   ) {blendType=Linear}
   n2 = Sequence Player 'Crow_Fly_Loop'!(PlayRate:float=Get WingFlapPlayRate() -> (WingFlapPlayRate:float)) -> (Pose) {sequence=&/Crow_Fly_Loop}
   n3 = Sequence Player 'Crow_Fly_Glide_Loop'() {sequence=&/Crow_Fly_Glide_Loop}
   n4 = "Sequence Player (additive) 'Crow_Fly_Pose_Down'"() {sequence=&/Crow_Fly_Pose_Down}
   Output Animation Pose(Result=Blend Poses by bool(
     BlendPose_0=Apply Additive(Base=@n3, Additive=@n4) {alphaBoolBlend {blendOption=Linear}}
     BlendPose_1=Apply Additive(Base=@n2, Additive=@n1) {alphaBoolBlend {blendInTime=0.4, blendOption=Linear, blendOutTime=0.4}}
     BlendTime_0=1
     BlendTime_1=0.5
     bActiveValue=Get bIsGliding() -> (bIsGliding:bool)
    ) {blendType=Linear})
  ]
 }
 TakeOff {
  path=AnimGraph.AnimGraphNode_StateMachine_1.CrowLoco.AnimStateNode_4.TakeOff
  nodes [Output Animation Pose(Result=Sequence Player 'Crow_Fly_Start'() {playRate=1.3, sequence=&/Crow_Fly_Start})]
 }
 Transition {path=AnimGraph.AnimGraphNode_StateMachine_1.CrowLoco.AnimStateTransitionNode_0.Transition, nodes [Result(bCanEnterTransition=true)]}
 Transition {
  path=AnimGraph.AnimGraphNode_StateMachine_1.CrowLoco.AnimStateTransitionNode_1.Transition
  nodes [
   Result(bCanEnterTransition=true)
   "float < float"!(A:double=Get Speed() -> (Speed:float), B:double=5) -> (ReturnValue:bool) {is=title K2Node_CallFunction}
  ]
 }
 Transition {
  path=AnimGraph.AnimGraphNode_StateMachine_1.CrowLoco.AnimStateTransitionNode_3.Transition
  nodes [Result(bCanEnterTransition=Get bIsLanding() -> (bIsLanding:bool))]
 }
 Transition {path=AnimGraph.AnimGraphNode_StateMachine_1.CrowLoco.AnimStateTransitionNode_4.Transition, nodes [Result()]}
 Transition {path=AnimGraph.AnimGraphNode_StateMachine_1.CrowLoco.AnimStateTransitionNode_5.Transition, nodes [Result(bCanEnterTransition=true)]}
 Transition {
  path=AnimGraph.AnimGraphNode_StateMachine_1.CrowLoco.AnimStateTransitionNode_6.Transition
  nodes [Result(bCanEnterTransition=Get bIsFlying() -> (bIsFlying:bool))]
 }
 Transition {path=AnimGraph.AnimGraphNode_StateMachine_1.CrowLoco.AnimStateTransitionNode_7.Transition, nodes [Result()]}
 EventGraph {nodes [Event BlueprintUpdateAnimation()]}
]
variables []
