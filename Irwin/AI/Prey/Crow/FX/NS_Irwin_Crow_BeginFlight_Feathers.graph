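# Niagara effect for the Irwin crow prey AI that sheds a couple of dark mesh feathers as the crow begins flight.
# A single burst emitter spawns randomly scaled and oriented feathers with outward, upward and slightly inherited velocity.
# Drag, gravity and age-ramped curl noise make them flutter down, while a random spin eases off over life.
# Late in life the feathers fade in alpha and shrink.
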NiagaraSystem /Irwin/AI/Prey/Crow/FX/NS_Irwin_Crow_BeginFlight_Feathers
dependencies {usedDataInterfaces [&NiagaraDataInterfaceCurve], usedDynamicInputs [], usedModules [&SystemState]}
emitters {feathers {
  emitterUpdateScript [
   &EmitterState(
    Enable Distance Culling=true
    Life Cycle Mode=::NewEnumerator1
    Loop Behavior=::NewEnumerator1
    Loop Delay=0.05
    Loop Duration=0.4
    Module.bUseMaxDistance=true
    Module.bUseMinDistance=false
    Scalability Mode=::NewEnumerator1
    UseLoopDelay=false
   )
   &SpawnBurst_Instantaneous(Spawn Count=2, Use Spawn Probability=false)
  ]
  particleSpawnScript [
   /Niagara/Modules/Spawn/Initialization/InitializeParticle(
    Color=<rgba 0 0 0 1>
    Lifetime=&FloatFromCurve(
     FloatCurve=DataInterface() {propertyValues {Curve {keys [
         {arriveTangent=0.0237, interpMode=RCIM_Cubic, leaveTangent=0.0237, tangentMode=RCTM_User, time=0, value=0.6}
         {arriveTangent=0.4051, interpMode=RCIM_Cubic, leaveTangent=0.4051, tangentMode=RCTM_User, time=1, value=1}
        ]}}}
     CurveIndex=&RandomRangeFloat(Minimum=0, Maximum=1)
     Scale Curve=3.5
    )
    Mesh Scale=&Multiply_VectorByFloat(Vector=&RandomRangeVector(Minimum=<0.6 0.6 1.5>, Maximum=<1 1 2.5>), Float=0.75)
    Module.Write Color=true
    Module.Write Lifetime=true
    Module.Write Mass=true
    Module.Write Material Random=true
    Module.Write Position=false
    Module.Write Scale=true
    Module.Write SpriteRotation=true
    Module.Write SpriteSize=true
    Position=&SimulationPosition()
    Sprite Rotation=&RandomRangeFloat(Minimum=0, Maximum=360)
   )
   /Niagara/Modules/Update/Mass/CalculateSizeAndRotationalInertiaByMass(Density by Material Type:ENiagaraMassByVolume=::NewEnumerator4, Depth=0, Height=5, Module.Manually Enter Density=false)
   &InitialMeshOrientation(
    Orientation Vector=<1 1 1>
    Rotation=&RandomRangeVector(Minimum=<0 0 0>, Maximum=<1 1 1>)
    Use Orientation Vector=true
    Use Rotation Vector=true
   )
   &ApplyInitialForces()
   /SetVariables(Particles.RandRotForce:NiagaraFloat=&RandomRangeFloat(Minimum=-6, Maximum=6)) {isSetParameters=true}
   &AddVelocityFromPoint(Module.Use Falloff Distance=false, Velocity Strength=&RandomRangeFloat(Minimum=10, Maximum=100))
   &AddVelocity(Velocity=&Multiply_Vector(
     A=<0 0 0>
     B=&VectorFromFloat(Value=&Multiply_Float(
       A=&RandomRangeFloat(Minimum=-75, Maximum=-300)
       B=&MakeCustomFloatFromBool(Bool=false, True Float=1, False Float=0.4)
      ))
    ))
   &AddVelocity(Velocity=&RandomRangeVector(Minimum=<0 0 25>, Maximum=<0 0 100>)) {module=AddVelocity001}
   &InheritVelocity(Module.Limit Speed=true, Velocity Scale=<0.1 0.1 0.1>)
  ]
  particleUpdateScript [
   &ScaleColor(
    Scale Alpha=&FloatFromCurve(FloatCurve=DataInterface() {propertyValues {Curve {keys [
         <at 0 1>
         {arriveTangent=-0.0304, interpMode=RCIM_Cubic, leaveTangent=-0.0304, tangentMode=RCTM_User, time=0.6, value=1}
         {arriveTangent=-2.4154, interpMode=RCIM_Cubic, leaveTangent=-2.4154, tangentMode=RCTM_User, time=1, value=0}
        ]}}})
    ScaleA=true
    ScaleRGB=true
   )
   &Drag(Drag=3, Use Linear Drag=true, Use Rotational Drag=false)
   &AccelerationForce(Acceleration=&RandomRangeVector(Minimum=<0 0 -10>, Maximum=<0 0 80>)) {enabled=false}
   &GravityForce(Gravity=&RandomRangeVector(Minimum=<0 0 -75>, Maximum=<0 0 -200>))
   &CurlNoiseForce(
    Noise Frequency=20
    Noise Strength=&FloatFromCurve(
     FloatCurve=DataInterface() {propertyValues {Curve {keys [
         {arriveTangent=2.5489, interpMode=RCIM_Cubic, leaveTangent=2.5489, tangentMode=RCTM_User, time=0, value=0.1}
         {arriveTangent=-0.0727, interpMode=RCIM_Cubic, leaveTangent=-0.0727, tangentMode=RCTM_User, time=1, value=1}
        ]}}}
     Scale Curve=800
    )
    Pan Noise=true
    Pan Noise Field=<0 0 0.5>
   )
   &SolveForcesAndVelocity(Clamp Velocity=false, Limit Acceleration=false)
   &SolveRotationalForcesAndVelocity()
   &ParticleState()
   &UpdateMeshOrientation(Rotation Vector=&Lerp_Vector(
     A=&Multiply_VectorByFloat(Vector=<0.1 0.1 1>, Float=$Particles.RandRotForce)
     B=<0 0 0>
     Alpha=&FloatFromCurve(FloatCurve=DataInterface() {propertyValues {Curve {keys [
          {arriveTangent=1.5391, interpMode=RCIM_Cubic, leaveTangent=1.5391, tangentMode=RCTM_User, time=0, value=0}
          {arriveTangent=1.9127, interpMode=RCIM_Cubic, leaveTangent=1.9127, tangentMode=RCTM_User, time=1, value=1}
         ]}}})
    ))
   &ScaleMeshSize(Scale Factor=&Multiply_VectorByFloat(
     Vector=<1 1 1>
     Float=&FloatFromCurve(FloatCurve=DataInterface() {propertyValues {Curve {keys [
          <at 0 1>
          {arriveTangent=-0.0008, interpMode=RCIM_Cubic, leaveTangent=-0.0008, tangentMode=RCTM_User, time=0.6, value=1}
          {arriveTangent=-3.4597, interpMode=RCIM_Cubic, leaveTangent=-3.4598, tangentMode=RCTM_User, time=1, value=0.2}
         ]}}})
    ))
  ]
  rendererClasses [&NiagaraMeshRendererProperties]
  renderers [{rendererClass=&NiagaraMeshRendererProperties, rendererIndex=0}]
 }}
summary {
 systemName=NS_Irwin_Crow_BeginFlight_Feathers
 emitters [{emitterName=feathers, rendererClasses [&NiagaraMeshRendererProperties]}]
 userVariables []
}
